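# Control Rig for the male medium Sahara Cloak face accessory part, driving corrective motion on hood, cape chain and shoulder bones.
# Head rotation relative to spine_05 weights additive hood-back offsets.
# Weighted transform constraints position chain bases from spine, neck and clavicles, then blend chain links toward the shared end bone.
# Arm angles drive right cape-shoulder correctives and a left shoulder tweak; IsSkydiving fades out the right high-arm corrective.
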ControlRigBlueprint /BRCosmetics/Characters/Player/Male/Medium/Bodies/M_MED_Sahara_Cloak/Meshes/Parts/M_MED_Sahara_Cloak_FaceAcc_CtrlRig
graphs [
 RigVMFunctionLibrary {links [], nodes []}
 RigVMModel {nodes [
   RigUnit_BeginExecution = &RigUnit_BeginExecution()
   n1 = &RigVMFunction_MathQuaternionToEuler(RotationOrder=ZYX, Value=&RigUnit_GetRelativeBoneTransform(Bone=head, Space=spine_05).Transform.Rotation)
   /Script/ControlRig.RigUnit_ModifyBoneTransforms(
    BoneToModify=((Transform=(Rotation=(X=0,Y=0,Z=0,W=1),Translation=(X=0,Y=0,Z=0),Scale3D=(X=1,Y=1,Z=1))))
    Mode=AdditiveLocal
    WeightMaximum=1
    Weight=&RigVMFunction_MathFloatRemap(SourceMaximum=-61, SourceMinimum=0, TargetMinimum=0, bClamp=true, Value=@n1.Result.X).Result
    ExecutePin=@RigUnit_BeginExecution.ExecutePin
   ) {is=struct RigVMUnitNode}
   > &RigUnit_ModifyBoneTransforms(
    BoneToModify=((Transform=(Rotation=(X=0,Y=0,Z=0,W=1),Translation=(X=0,Y=0,Z=0),Scale3D=(X=1,Y=1,Z=1))))
    Mode=AdditiveLocal
    WeightMaximum=1
    Weight=&RigVMFunction_MathFloatRemap(SourceMaximum=70, SourceMinimum=0, TargetMinimum=0, bClamp=true, Value=@n1.Result.X).Result
   )
   n2 = &RigVMFunction_MathFloatRemap(SourceMaximum=-48, SourceMinimum=-2, TargetMinimum=0, bClamp=true, Value=@n1.Result.Z)
   > &RigUnit_ModifyBoneTransforms(
    BoneToModify=((Bone="dyn_hoodBack_r",Transform=(Rotation=(X=0,Y=0,Z=0,W=1),Translation=(X=5,Y=2,Z=10),Scale3D=(X=1,Y=1,Z=1))))
    Mode=AdditiveLocal
    WeightMaximum=1
    WeightMinimum=0
    Weight=@n2.Result
   )
   > &RigUnit_ModifyBoneTransforms(
    BoneToModify=((Bone="dyn_hoodBack",Transform=(Rotation=(X=0,Y=0,Z=0,W=1),Translation=(X=10,Y=-2,Z=0),Scale3D=(X=1,Y=1,Z=1))))
    Mode=AdditiveLocal
    WeightMaximum=1
    Weight=@n2.Result
   )
   > &RigUnit_ModifyBoneTransforms(
    BoneToModify=((Bone="dyn_hoodBack_l",Transform=(Rotation=(X=0,Y=0,Z=0,W=1),Translation=(X=5,Y=2,Z=6),Scale3D=(X=1,Y=1,Z=1))))
    Mode=AdditiveLocal
    WeightMaximum=1
    Weight=@n2.Result
   )
   > /Script/ControlRig.RigUnit_TransformConstraint(
    BaseTransform=(Rotation=(X=0,Y=0,Z=0,W=1),Translation=(X=0,Y=0,Z=0),Scale3D=(X=1,Y=1,Z=1))
    BaseTransformSpace=GlobalSpace
    Bone=dyn_capeChain_01_Base_r
    Targets="((Transform=(Rotation=(X=0,Y=0,Z=0,W=1),Translation=(X=0,Y=0,Z=0),Scale3D=(X=1,Y=1,Z=1)),Weight=0.0716,bMaintainOffset=True,Filter=(TranslationFilter=(bX=True,bY=True,bZ=True),RotationFilter=(bX=True,bY=True,bZ=True),ScaleFilter=(bX=True,bY=True,bZ=True))) ...(truncated)"
    bUseInitialTransforms=True
    Targets.0.Transform=&RigUnit_GetBoneTransform(Bone=clavicle_cape_r).Transform
    Targets.1.Transform=&RigUnit_GetBoneTransform(Bone=neck_02).Transform
    Targets.2.Transform=&RigUnit_GetBoneTransform(Bone=spine_03).Transform
    Targets.3.Transform=&RigUnit_GetBoneTransform(Bone=spine_04).Transform
    Targets.4.Transform=&RigUnit_GetBoneTransform(Bone=spine_05).Transform
    Targets.5.Transform=&RigUnit_GetBoneTransform(Bone=clavicle_r).Transform
   ) {is=struct RigVMUnitNode}
   > &RigUnit_TransformConstraint(
    BaseTransform=(Rotation=(X=0,Y=0,Z=0,W=1),Translation=(X=0,Y=0,Z=0),Scale3D=(X=1,Y=1,Z=1))
    BaseTransformSpace=GlobalSpace
    Bone=dyn_capeChain_01_Base_l
    Targets="((Transform=(Rotation=(X=0,Y=0,Z=0,W=1),Translation=(X=0,Y=0,Z=0),Scale3D=(X=1,Y=1,Z=1)),Weight=0.0716,bMaintainOffset=True,Filter=(TranslationFilter=(bX=True,bY=True,bZ=True),RotationFilter=(bX=True,bY=True,bZ=True),ScaleFilter=(bX=True,bY=True,bZ=True))) ...(truncated)"
    bUseInitialTransforms=True
    Targets.1.Transform=&RigUnit_GetBoneTransform(Bone=neck_02).Transform
    Targets.2.Transform=&RigUnit_GetBoneTransform(Bone=spine_03).Transform
    Targets.3.Transform=&RigUnit_GetBoneTransform(Bone=spine_04).Transform
    Targets.5.Transform=&RigUnit_GetBoneTransform(Bone=clavicle_l).Transform
    Targets.4.Transform=&RigUnit_GetBoneTransform(Bone=spine_05).Transform
    Targets.0.Transform=&RigUnit_GetBoneTransform(Bone=clavicle_cape_l).Transform
   )
   n3 = &RigUnit_GetBoneTransform(Bone=dyn_capeChainEnd)
   n4 = &RigUnit_GetBoneTransform(Bone=dyn_capeChain_01_Base_r)
   > &RigUnit_TransformConstraint(
    BaseTransform=(Rotation=(X=0,Y=0,Z=0,W=1),Translation=(X=0,Y=0,Z=0),Scale3D=(X=1,Y=1,Z=1))
    BaseTransformSpace=GlobalSpace
    Bone=dyn_capeChain_02_r
    Targets="((Transform=(Rotation=(X=0,Y=0,Z=0,W=1),Translation=(X=0,Y=0,Z=0),Scale3D=(X=1,Y=1,Z=1)),Weight=0.4,bMaintainOffset=True,Filter=(TranslationFilter=(bX=True,bY=True,bZ=True),RotationFilter=(bX=True,bY=True,bZ=True),ScaleFilter=(bX=True,bY=True,bZ=True))),(T ...(truncated)"
    bUseInitialTransforms=True
    Targets.0.Transform=@n3.Transform
    Targets.1.Transform=@n4.Transform
   )
   > &RigUnit_TransformConstraint(
    BaseTransform=(Rotation=(X=0,Y=0,Z=0,W=1),Translation=(X=0,Y=0,Z=0),Scale3D=(X=1,Y=1,Z=1))
    BaseTransformSpace=GlobalSpace
    Bone=dyn_capeChain_03_r
    Targets="((Transform=(Rotation=(X=0,Y=0,Z=0,W=1),Translation=(X=0,Y=0,Z=0),Scale3D=(X=1,Y=1,Z=1)),Weight=0.7,bMaintainOffset=True,Filter=(TranslationFilter=(bX=True,bY=True,bZ=True),RotationFilter=(bX=True,bY=True,bZ=True),ScaleFilter=(bX=True,bY=True,bZ=True))),(T ...(truncated)"
    bUseInitialTransforms=True
    Targets.0.Transform=@n3.Transform
    Targets.1.Transform=@n4.Transform
   )
   n5 = &RigUnit_GetBoneTransform(Bone=dyn_capeChain_01_Base_l)
   > &RigUnit_TransformConstraint(
    BaseTransform=(Rotation=(X=0,Y=0,Z=0,W=1),Translation=(X=0,Y=0,Z=0),Scale3D=(X=1,Y=1,Z=1))
    BaseTransformSpace=GlobalSpace
    Bone=dyn_capeChain_02_l
    Targets="((Transform=(Rotation=(X=0,Y=0,Z=0,W=1),Translation=(X=0,Y=0,Z=0),Scale3D=(X=1,Y=1,Z=1)),Weight=0.4,bMaintainOffset=True,Filter=(TranslationFilter=(bX=True,bY=True,bZ=True),RotationFilter=(bX=True,bY=True,bZ=True),ScaleFilter=(bX=True,bY=True,bZ=True))),(T ...(truncated)"
    bUseInitialTransforms=True
    Targets.0.Transform=@n3.Transform
    Targets.1.Transform=@n5.Transform
   )
   > &RigUnit_TransformConstraint(
    BaseTransform=(Rotation=(X=0,Y=0,Z=0,W=1),Translation=(X=0,Y=0,Z=0),Scale3D=(X=1,Y=1,Z=1))
    BaseTransformSpace=GlobalSpace
    Bone=dyn_capeChain_03_l
    Targets="((Transform=(Rotation=(X=0,Y=0,Z=0,W=1),Translation=(X=0,Y=0,Z=0),Scale3D=(X=1,Y=1,Z=1)),Weight=0.7,bMaintainOffset=True,Filter=(TranslationFilter=(bX=True,bY=True,bZ=True),RotationFilter=(bX=True,bY=True,bZ=True),ScaleFilter=(bX=True,bY=True,bZ=True))),(T ...(truncated)"
    bUseInitialTransforms=True
    Targets.0.Transform=@n3.Transform
    Targets.1.Transform=@n5.Transform
   )
   n6 = /Script/RigVM.RigVMFunction_MathFloatAbs(Value=&RigVMFunction_MathQuaternionToEuler(RotationOrder=ZYX, Value=&RigUnit_GetRelativeBoneTransform(Bone=upperarm_r, Space=spine_05).Transform.Rotation).Result.Z) {is=struct RigVMUnitNode, template="Absolute::Execute(in Value,out Result)"}
   > &RigUnit_ModifyBoneTransforms(
    BoneToModify="((Bone=\"dyn_capeFrontShoulder_r\",Transform=(Rotation=(X=0,Y=0,Z=0,W=1),Translation=(X=-1.9,Y=-3.5,Z=-1.25),Scale3D=(X=1,Y=1,Z=1))),(Bone=\"dyn_capeTopShoulder_r\",Transform=(Rotation=(X=0,Y=0,Z=0,W=1),Translation=(X=0,Y=0,Z=0),Scale3D=(X=1,Y=1,Z=1))),(Bone=\" ...(truncated)"
    Mode=AdditiveLocal
    WeightMaximum=1
    Weight=&RigVMFunction_MathFloatMul(
     B=1
     A=&RigVMFunction_MathFloatRemap(SourceMaximum=123, SourceMinimum=30, TargetMinimum=0.15, bClamp=true, Value=@n6.Result).Result
    ) {template="Multiply::Execute(in A,in B,out Result)"}.Result
    BoneToModify.0.Transform.Rotation=&RigVMFunction_MathQuaternionFromEuler(Euler=(X=0,Y=0,Z=0)).Result
    BoneToModify.1.Transform.Rotation=&RigVMFunction_MathQuaternionFromEuler(Euler=(X=0,Y=0,Z=-45)).Result
    BoneToModify.2.Transform.Rotation=&RigVMFunction_MathQuaternionFromEuler(Euler=(X=0,Y=0,Z=-45)).Result
   )
   n7 = RigVMVariableNode() {variable=IsSkydiving}
   > &RigUnit_ModifyBoneTransforms(
    BoneToModify="((Bone=\"dyn_capeBackShoulder_Base_r\",Transform=(Rotation=(X=0,Y=0,Z=0,W=1),Translation=(X=0,Y=-5,Z=15),Scale3D=(X=1,Y=1,Z=1))),(Bone=\"dyn_capeTopShoulder_Base_r\",Transform=(Rotation=(X=0,Y=0,Z=0,W=1),Translation=(X=0,Y=0,Z=6),Scale3D=(X=1,Y=1,Z=1))),(Bone= ...(truncated)"
    Mode=AdditiveLocal
    WeightMaximum=1
    Weight=&RigVMFunction_MathFloatMul(
     A=&RigVMFunction_MathFloatRemap(SourceMaximum=141, SourceMinimum=123, TargetMinimum=0, bClamp=true, Value=@n6.Result).Result
     B=/Script/RigVM.RigVMFunction_MathFloatSub(A=1, B=@n7.Value) {is=struct RigVMUnitNode, template="Subtract::Execute(in A,in B,out Result)"}.Result
    ) {template="Multiply::Execute(in A,in B,out Result)"}.Result
    BoneToModify.0.Transform.Rotation=&RigVMFunction_MathQuaternionFromEuler(Euler=(X=-20,Y=20,Z=-10)).Result
    BoneToModify.1.Transform.Rotation=&RigVMFunction_MathQuaternionFromEuler(Euler=(X=0,Y=0,Z=-30)).Result
    BoneToModify.2.Transform.Rotation=&RigVMFunction_MathQuaternionFromEuler(Euler=(X=0,Y=0,Z=-25)).Result
   )
   n8 = &RigVMFunction_MathQuaternionToEuler(RotationOrder=ZYX, Value=&RigUnit_GetRelativeBoneTransform(Bone=upperarm_l, Space=clavicle_l).Transform.Rotation)
   > &RigUnit_ModifyBoneTransforms(
    BoneToModify=((Bone="dyn_shoulderScale_l",Transform=(Rotation=(X=0,Y=0,Z=0,W=1),Translation=(X=0,Y=0,Z=0),Scale3D=(X=1,Y=1,Z=1))))
    Mode=AdditiveLocal
    WeightMaximum=1
    Weight=&RigVMFunction_MathFloatRemap(
     SourceMaximum=27
     SourceMinimum=74
     TargetMinimum=0
     bClamp=true
     Value=&RigVMFunction_MathFloatAbs(Value=@n8.Result.Z) {template="Absolute::Execute(in Value,out Result)"}.Result
    ).Result
    BoneToModify.0.Transform.Rotation=&RigVMFunction_MathQuaternionFromEuler(Euler=(X=-50,Y=10,Z=-10)).Result
   )
   &RigVMFunction_MathFloatRemap(SourceMaximum=1, bClamp=true, Value=@n1.Result.Y)
  ]}
]
hierarchy [
 root {loc=<0 0 0>, rot=<0 0 0 1>}
 pelvis {parent=root, loc=<0 -0.7071 93.1191>, rot=<0.0127 0.707 -0.0127 -0.707>}
 spine_01 {parent=pelvis, loc=<6.7494 -0.6388 0>, rot=<0 0 0.034 -0.9994>}
 spine_02 {parent=spine_01, loc=<7.2518 -0.035 0>, rot=<0 0 0.0238 -0.9997>}
 spine_03 {parent=spine_02, loc=<6.0104 0 0>, rot=<0 0 0.0149 -0.9999>}
 spine_04 {parent=spine_03, loc=<8.1024 0 0>, rot=<0 0 -0.0132 -0.9999>}
 spine_05 {parent=spine_04, loc=<10.0146 0 0>, rot=<0 0 -0.0015 -1>}
 neck_01 {parent=spine_05, loc=<11.3138 0.1529 0>, rot=<0 0 -0.2003 -0.9797>}
 neck_02 {parent=neck_01, loc=<4.5627 0 0>, rot=<0 0 0.0504 -0.9987>}
 head {parent=neck_02, loc=<5.3914 0 0>, rot=<0 0 0.115 -0.9934>}
 dyn_hoodBack_Base_r {parent=neck_02, loc=<3.6414 -6.0617 7.1117>, rot=<-0.6805 0.3247 -0.2595 -0.6035>}
 dyn_hoodBack_r {parent=dyn_hoodBack_Base_r, loc=<0 0 0>, rot=<0 0 0 1>}
 dyn_hoodBack_Base {parent=neck_02, loc=<2.5535 -9.271 0.0312>, rot=<0.6455 -0.7634 0.0057 -0.0214>}
 dyn_hoodBack {parent=dyn_hoodBack_Base, loc=<0 0 0>, rot=<0 0 0 1>}
 dyn_hoodBack_Base_l {parent=neck_02, loc=<3.6416 -6.0617 -7.1117>, rot=<0.6589 -0.3337 -0.2481 -0.6268>}
 dyn_hoodBack_l {parent=dyn_hoodBack_Base_l, loc=<0 0 0>, rot=<0 0 0 1>}
 Dyn_Head {parent=neck_02, loc=<5.3914 0 0>, rot=<0 0 0.115 -0.9934>}
 dyn_hoodTop_Base {parent=Dyn_Head, loc=<11.1741 10.072 0.2627>, rot=<0.0086 0.0047 0.8847 0.4661>}
 dyn_hoodTop {parent=dyn_hoodTop_Base, loc=<0 0 0>, rot=<0 0 0 1>}
 dyn_hoodSide_Base_r {parent=Dyn_Head, loc=<-0.4725 1.5259 9.1634>, rot=<-0.4676 -0.5153 0.5927 0.4056>}
 dyn_hoodSide_r {parent=dyn_hoodSide_Base_r, loc=<0 0 0>, rot=<0 0 0 1>}
 dyn_hoodSide_Base_l {parent=Dyn_Head, loc=<-0.4729 1.5259 -9.1634>, rot=<-0.4056 -0.5927 0.5153 0.4676>}
 dyn_hoodSide_l {parent=dyn_hoodSide_Base_l, loc=<0 0 0>, rot=<0 0 0 1>}
 dyn_hoodPoint {parent=Dyn_Head, loc=<13.5725 -3.1228 -0.4722>, rot=<0.6878 -0.7251 0.0234 0.0219>}
 clavicle_l {parent=spine_05, loc=<4.3575 -0.1466 -2.3223>, rot=<0.0465 -0.7289 -0.009 -0.6829>}
 upperarm_l {parent=clavicle_l, loc=<15.5111 0 0>, rot=<0.0226 -0.3234 -0.0088 -0.9459>}
 lowerarm_l {parent=upperarm_l, loc=<26.6514 0 0>, rot=<0.0391 -0.0099 -0.2453 -0.9686>}
 deform_elbow_out_base_l {parent=lowerarm_l, loc=<0 -4 0>, rot=<0 0 0 -1>}
 deform_elbow_out_l {parent=deform_elbow_out_base_l, loc=<0 0 0>, rot=<0 0 0 1>}
 upperarm_twist_01_l {parent=upperarm_l, loc=<0.0041 0 0>, rot=<0 0 0 -1>}
 deform_upperarm_bk_base_l {parent=upperarm_twist_01_l, loc=<1.0813 -7.2193 -1.9453>, rot=<0 0 0 -1>}
 deform_upperarm_bk_l {parent=deform_upperarm_bk_base_l, loc=<0 0 0>, rot=<0 0 0 1>}
 deform_upperarm_fr_base_l {parent=upperarm_twist_01_l, loc=<-0.3691 7.0506 -2.765>, rot=<0 0 0 -1>}
 deform_upperarm_fr_l {parent=deform_upperarm_fr_base_l, loc=<0 0 0>, rot=<0 0 0 1>}
 deform_upperarm_up_base_l {parent=upperarm_twist_01_l, loc=<-3.2216 -1.1391 3.3311>, rot=<0 0 0 -1>}
 deform_upperarm_up_l {parent=deform_upperarm_up_base_l, loc=<0 0 0>, rot=<0 0 0 1>}
 upperarm_twist_02_l {parent=upperarm_l, loc=<13.7941 0 0>, rot=<0 0 0 -1>}
 clavicle_cape_l {parent=clavicle_l, loc=<0 0.0001 0.0003>, rot=<0 0 0 -1>}
 cape_01_l {parent=clavicle_cape_l, loc=<13.08 -4.0816 7.516>, rot=<0.3413 -0.5667 0.2993 -0.6876>}
 cape_02_l {parent=cape_01_l, loc=<6.892 0 0>, rot=<-0.0025 0.0083 -0.2906 -0.9568>}
 cape_03_l {parent=cape_02_l, loc=<11.8886 0 0>, rot=<0 0 -0.1305 -0.9914>}
 cape_04_l {parent=cape_03_l, loc=<23.3505 0 0>, rot=<0 0 -0.0306 -0.9995>}
 cape_05_l {parent=cape_04_l, loc=<20.0893 0 0>, rot=<0 0 -0.008 -1>}
 cape_06_l {parent=cape_05_l, loc=<18.5056 0 0>, rot=<0 0 -0.0041 -1>}
 cape_07_l {parent=cape_06_l, loc=<13.5867 0 0>, rot=<0 0 0.0015 -1>}
 cape_08_l {parent=cape_07_l, loc=<15.1001 0 0>, rot=<0 0 -0.0027 -1>}
 dyn_capeBackShoulder_Base_l {parent=clavicle_cape_l, loc=<8.0681 -5.9931 9.6931>, rot=<0.5954 -0.1376 -0.0578 -0.7894>}
 dyn_capeBackShoulder_l {parent=dyn_capeBackShoulder_Base_l, loc=<0 0 0>, rot=<0 0 0 1>}
 dyn_capeTopShoulder_Base_l {parent=clavicle_cape_l, loc=<8.4872 -3.1592 9.4987>, rot=<0.7388 -0.026 -0.1127 -0.6639>}
 dyn_capeTopShoulder_l {parent=dyn_capeTopShoulder_Base_l, loc=<0 0 0>, rot=<0 0 0 1>}
 dyn_capeFrontShoulder_Base_l {parent=clavicle_cape_l, loc=<6.0819 5.2175 6.6089>, rot=<0.8836 0.0378 -0.1728 -0.4335>}
 dyn_capeFrontShoulder_l {parent=dyn_capeFrontShoulder_Base_l, loc=<0 0 0>, rot=<0 0 0 1>}
 clavicle_r {parent=spine_05, loc=<4.3575 -0.1466 2.3223>, rot=<0.7289 0.0465 -0.6829 0.009>}
 upperarm_r {parent=clavicle_r, loc=<-15.5111 0 0>, rot=<0.0226 -0.3234 -0.0088 -0.9459>}
 lowerarm_r {parent=upperarm_r, loc=<-26.6514 0 0>, rot=<0.0391 -0.0099 -0.2453 -0.9686>}
 deform_elbow_out_base_r {parent=lowerarm_r, loc=<0 4 -0.0001>, rot=<0 0 0 -1>}
 deform_elbow_out_r {parent=deform_elbow_out_base_r, loc=<0 0 0>, rot=<0 0 0 1>}
 upperarm_twist_01_r {parent=upperarm_r, loc=<-0.0041 0 0>, rot=<0 0 0 -1>}
 deform_upperarm_bk_base_r {parent=upperarm_twist_01_r, loc=<-1.0814 7.2192 1.9456>, rot=<0 0 0 -1>}
 deform_upperarm_bk_r {parent=deform_upperarm_bk_base_r, loc=<0 0 0>, rot=<0 0 0 1>}
 deform_upperarm_fr_base_r {parent=upperarm_twist_01_r, loc=<0.3689 -7.0506 2.7652>, rot=<0 0 0 -1>}
 deform_upperarm_fr_r {parent=deform_upperarm_fr_base_r, loc=<0 0 0>, rot=<0 0 0 1>}
 deform_upperarm_up_base_r {parent=upperarm_twist_01_r, loc=<3.2214 1.1391 -3.3308>, rot=<0 0 0 1>}
 deform_upperarm_up_r {parent=deform_upperarm_up_base_r, loc=<0 0 0>, rot=<0 0 0 -1>}
 upperarm_twist_02_r {parent=upperarm_r, loc=<-13.7941 0 0>, rot=<0 0 0 -1>}
 clavicle_cape_r {parent=clavicle_r, loc=<0 0 0>, rot=<0 0 0 -1>}
 cape_01_r {parent=clavicle_cape_r, loc=<-13.0799 4.0816 -7.5166>, rot=<0.5667 0.3413 0.6876 0.2993>}
 cape_02_r {parent=cape_01_r, loc=<6.8922 -0.0002 0>, rot=<0.0025 -0.0083 -0.2906 -0.9568>}
 cape_03_r {parent=cape_02_r, loc=<11.8886 0 0>, rot=<0 0 -0.1305 -0.9914>}
 cape_04_r {parent=cape_03_r, loc=<23.3501 0.0001 0>, rot=<0 0 -0.0306 -0.9995>}
 cape_05_r {parent=cape_04_r, loc=<20.0898 0 0.0001>, rot=<0 0 -0.008 -1>}
 cape_06_r {parent=cape_05_r, loc=<18.5056 -0.0001 -0.0001>, rot=<0 0 -0.0041 -1>}
 cape_07_r {parent=cape_06_r, loc=<13.5867 0 0.0001>, rot=<0 0 0.0015 -1>}
 cape_08_r {parent=cape_07_r, loc=<15.1001 0.0001 -0.0001>, rot=<0 0 -0.0027 -1>}
 dyn_capeFrontShoulder_Base_r {parent=clavicle_cape_r, loc=<-6.0818 -5.2176 -6.6095>, rot=<0.0172 -0.8265 -0.5345 0.176>}
 dyn_capeFrontShoulder_r {parent=dyn_capeFrontShoulder_Base_r, loc=<0 0 0>, rot=<0 0 0 1>}
 dyn_capeTopShoulder_Base_r {parent=clavicle_cape_r, loc=<-8.4872 3.1591 -9.4985>, rot=<-0.0269 -0.733 -0.6703 0.1124>}
 dyn_capeTopShoulder_r {parent=dyn_capeTopShoulder_Base_r, loc=<0 0 0>, rot=<0 0 0 1>}
 dyn_capeBackShoulder_Base_r {parent=clavicle_cape_r, loc=<-8.0681 5.993 -9.6935>, rot=<0.1335 0.6463 0.7484 -0.0668>}
 dyn_capeBackShoulder_r {parent=dyn_capeBackShoulder_Base_r, loc=<0 0 0>, rot=<0 0 0 1>}
 attach_backpack {parent=spine_05, loc=<-0.123 -13.4727 0>, rot=<0 -0.7071 0 -0.7071>}
 attach_cape {parent=spine_05, loc=<9.017 -11.714 0>, rot=<0.0284 -0.7065 -0.0284 -0.7065>}
 backpack_root {parent=spine_05, loc=<9.0735 -9.9874 0>, rot=<0.0284 -0.7065 -0.0284 -0.7065>}
 cape_mid_01_l {parent=backpack_root, loc=<6.2329 -0.2606 -0.4394>, rot=<0.1329 -0.6858 0.1296 -0.7037>}
 cape_mid_02_l {parent=cape_mid_01_l, loc=<12.8326 0 0>, rot=<0 0 -0.1478 -0.989>}
 cape_mid_03_l {parent=cape_mid_02_l, loc=<23.0451 0 0>, rot=<0 0 -0.0085 -1>}
 cape_mid_04_l {parent=cape_mid_03_l, loc=<20.3244 0 0>, rot=<0 0 -0.0139 -0.9999>}
 cape_mid_05_l {parent=cape_mid_04_l, loc=<18.7251 0 0>, rot=<0 0 0.0005 -1>}
 cape_mid_06_l {parent=cape_mid_05_l, loc=<13.6769 0 0>, rot=<0 0 0.0016 -1>}
 cape_mid_07_l {parent=cape_mid_06_l, loc=<14.7332 0 0>, rot=<0 0 -0.0176 -0.9998>}
 cape_mid_01_r {parent=backpack_root, loc=<-6.2329 -0.2606 -0.4389>, rot=<0.1296 -0.7037 0.1329 -0.6858>}
 cape_mid_02_r {parent=cape_mid_01_r, loc=<12.8333 -0.0003 0>, rot=<0 0 -0.1478 -0.989>}
 cape_mid_03_r {parent=cape_mid_02_r, loc=<23.0446 0 0>, rot=<0 0 -0.0085 -1>}
 cape_mid_04_r {parent=cape_mid_03_r, loc=<20.3247 0 0>, rot=<0 0 -0.0139 -0.9999>}
 cape_mid_05_r {parent=cape_mid_04_r, loc=<18.7251 0 0>, rot=<0 0 0.0005 -1>}
 cape_mid_06_r {parent=cape_mid_05_r, loc=<13.6769 -0.0001 0>, rot=<0 0 0.0016 -1>}
 cape_mid_07_r {parent=cape_mid_06_r, loc=<14.7332 0 0>, rot=<0 0 -0.0176 -0.9998>}
 dyn_capeChain_01_Base_l {parent=spine_05, loc=<-2.3345 14.456 -7.9749>, rot=<0.5447 -0.0278 0.8368 0.0486>}
 dyn_capeChain_01_l {parent=dyn_capeChain_01_Base_l, loc=<0 0 0>, rot=<0 0 0 1>}
 dyn_capeChain_02_l {parent=dyn_capeChain_01_Base_l, loc=<2.8581 -0.4232 -0.7749>, rot=<-0.0007 0.0624 -0.0342 -0.9975>}
 dyn_capeChain_01_Base_r {parent=spine_05, loc=<-2.4643 14.0139 8.2741>, rot=<-0.0739 -0.8358 0.0191 -0.5437>}
 dyn_capeChain_01_r {parent=dyn_capeChain_01_Base_r, loc=<0 0 0>, rot=<0 0 0 1>}
 dyn_capeChain_02_r {parent=dyn_capeChain_01_Base_r, loc=<2.9858 -0.1884 -0.6115>, rot=<0.0533 0.0533 -0.0133 -0.9971>}
 dyn_capeChainStart {parent=spine_05, loc=<-2.3994 14.2349 0.1496>, rot=<0.6473 -0.0644 0.071 -0.7562>}
 dyn_capeChainEnd {parent=dyn_capeChainStart, loc=<-3.6251 -0.0001 0.0001>, rot=<0 0 0 1>}
 deform_pec_aim_base_l {parent=spine_04, loc=<7.8732 12.6388 -2.4123>, rot=<0.0102 -0.4602 0.3979 -0.7936>}
 deform_pec_aim_l {parent=deform_pec_aim_base_l, loc=<0 0 0>, rot=<0 0 0 1>}
 deform_pec_base_l {parent=deform_pec_aim_l, loc=<10.766 0 0>, rot=<0 0 0 1>}
 deform_pec_l {parent=deform_pec_base_l, loc=<0.055 0 0>, rot=<0 0 0 1>}
 deform_pec_aim_base_r {parent=spine_04, loc=<7.8727 12.6388 2.4123>, rot=<-0.4602 -0.0102 0.7936 0.3979>}
 deform_pec_aim_r {parent=deform_pec_aim_base_r, loc=<0 0 0>, rot=<0 0 0 1>}
 deform_pec_base_r {parent=deform_pec_aim_r, loc=<-10.7662 -0.0006 -0.0007>, rot=<0 0 0 1>}
 deform_pec_r {parent=deform_pec_base_r, loc=<0 0 0>, rot=<0 0 0 1>}
 thigh_l {parent=pelvis, loc=<-2.5693 -0.0109 -7.7501>, rot=<0.0012 0.0572 0.0155 -0.9982>}
 calf_l {parent=thigh_l, loc=<-39.0484 0 0>, rot=<0.0006 -0.0161 -0.0391 -0.9991>}
 calf_twist_02_l {parent=calf_l, loc=<-18.8064 -0.0967 0.0397>, rot=<0.0237 0.0095 -0.0035 -0.9997>}
 calf_twist_01_l {parent=calf_l, loc=<-35.813 -0.0967 0.0397>, rot=<0.0237 0.0095 -0.0035 -0.9997>}
 deform_knee_in_base_l {parent=calf_l, loc=<1.298 -5.2957 -0.0496>, rot=<0 0 0 -1>}
 deform_knee_in_l {parent=deform_knee_in_base_l, loc=<0 0 0>, rot=<0 0 0 1>}
 thigh_r {parent=pelvis, loc=<-2.5691 -0.0109 7.7501>, rot=<0.0572 -0.0012 0.9982 0.0155>}
 calf_r {parent=thigh_r, loc=<39.0484 0 0>, rot=<0.0006 -0.0161 -0.0391 -0.9991>}
 calf_twist_02_r {parent=calf_r, loc=<18.8064 0.0967 -0.0397>, rot=<0.0237 0.0095 -0.0035 -0.9997>}
 calf_twist_01_r {parent=calf_r, loc=<35.813 0.0967 -0.0397>, rot=<0.0237 0.0095 -0.0035 -0.9997>}
 deform_knee_in_base_r {parent=calf_r, loc=<-1.298 5.2957 0.0496>, rot=<0 0 0 -1>}
 deform_glute_base_l {parent=pelvis, loc=<-8.8064 -8.8222 -8.4287>, rot=<0.0275 0.0501 -0.4472 -0.8926>}
 deform_glute_base_r {parent=pelvis, loc=<-8.8062 -8.8222 8.4287>, rot=<0.0501 -0.0275 0.8926 -0.4472>}
 deform_glute_r {parent=deform_glute_base_r, loc=<0 0 0>, rot=<0 0 0 1>}
 ik_foot_root {parent=root, loc=<0 0 0>, rot=<0 0 0 1>}
 ik_hand_root {parent=root, loc=<0 0 0>, rot=<0 0 0 1>}
 ik_hand_gun {parent=ik_hand_root, loc=<-51.9727 4.0883 102.6935>, rot=<0.6084 0.1793 0.4613 -0.6205>}
 dyn_crouch_collision {parent=root, loc=<0 0 0>, rot=<0 0 0 1>}
]
variables [IsSkydiving {type=double, default=0}]
