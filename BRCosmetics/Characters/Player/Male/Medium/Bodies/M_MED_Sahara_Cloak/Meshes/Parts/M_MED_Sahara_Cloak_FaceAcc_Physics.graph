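# Physics asset for the face-accessory part of the male medium Sahara Cloak body, set up for dynamic motion of its hood.
# Capsules on upper spine, clavicles, neck and head form the body volumes, joined by free-rotation constraints.
# Tiny base capsules anchor hood top and sides to the head, and three back pieces to the neck.
# Each larger hood capsule hangs from its base with individually tuned, mostly tight swing and twist limits.
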PhysicsAsset /BRCosmetics/Characters/Player/Male/Medium/Bodies/M_MED_Sahara_Cloak/Meshes/Parts/M_MED_Sahara_Cloak_FaceAcc_Physics
bodies {
 clavicle_l [{shapeType=Capsule, length=15, radius=12, shapeName=clavicle_l_capsule, center=<15.7783 0 -0.4243>, rotation=<rot 90 0 0>}]
 clavicle_r [{shapeType=Capsule, length=15, radius=12, shapeName=clavicle_r_capsule, center=<-15.7783 0 0.3393>, rotation=<rot 90 0 0>}]
 dyn_hoodBack [{shapeType=Capsule, length=12, radius=5, shapeName=dyn_hoodBack_capsule, center=<5.3346 0.8418 0>}]
 dyn_hoodBack_Base [{shapeType=Capsule, length=0.5, radius=0.5, shapeName=dyn_hoodBack_Base_capsule}]
 dyn_hoodBack_Base_l [{shapeType=Capsule, length=0.5, radius=0.5, shapeName=dyn_hoodBack_Base_l_capsule}]
 dyn_hoodBack_Base_r [{shapeType=Capsule, length=0.5, radius=0.5, shapeName=dyn_hoodBack_Base_r_capsule}]
 dyn_hoodBack_l [{
   shapeType=Capsule
   length=10
   radius=5
   shapeName=dyn_hoodBack_l_capsule
   center=<1.9243 3.1713 -0.0541>
   rotation=<rot -17.3036 0.7447 7.0154>
  }]
 dyn_hoodBack_r [{
   shapeType=Capsule
   length=10
   radius=5
   shapeName=dyn_hoodBack_r_capsule
   center=<2.9268 2.6838 -0.2938>
   rotation=<rot 22.3317 -1.7726 -11.8039>
  }]
 dyn_hoodSide_Base_l [{shapeType=Capsule, length=0.5, radius=0.5, shapeName=dyn_hoodSide_Base_l_capsule}]
 dyn_hoodSide_Base_r [{shapeType=Capsule, length=0.5, radius=0.5, shapeName=dyn_hoodSide_Base_r_capsule}]
 dyn_hoodSide_l [{shapeType=Capsule, length=8, radius=6.5, shapeName=dyn_hoodSide_l_capsule, center=<3 2.7357 0>, rotation=<rot 90 0 0>}]
 dyn_hoodSide_r [{shapeType=Capsule, length=8, radius=6.5, shapeName=dyn_hoodSide_r_capsule, center=<3 -2.3683 0>, rotation=<rot 90 0 0>}]
 dyn_hoodTop [{shapeType=Capsule, length=6, radius=6, shapeName=dyn_hoodTop_capsule, center=<3 -5.776 0>, rotation=<rot 90 0 0>}]
 dyn_hoodTop_Base [{shapeType=Capsule, length=0.5, radius=0.5, shapeName=dyn_hoodTop_Base_capsule}]
 head [{shapeType=Capsule, length=9.4771, radius=4.0709, center=<0.6463 1.6911 0>, rotation=<rot 90 0 -14.7504>}]
 neck_02 [{shapeType=Capsule, length=0.505, radius=0.505, rotation=<rot 0 -90 0>}]
 spine_05 [
  {shapeType=Capsule, length=16, radius=11.5, shapeName=spine_05_capsule, center=<3.8673 -5.7269 0.7793>}
  {shapeType=Capsule, length=25, radius=12, shapeName=spine_05_capsule, center=<1.72 4.6085 0>}
 ]
}
constraints [
 {
  bone1Name=head
  bone2Name=neck_02
  swing1LimitDegrees=45
  swing1Motion=Free
  swing2LimitDegrees=45
  swing2Motion=Free
  twistLimitDegrees=45
  twistMotion=Free
 }
 {
  bone1Name=dyn_hoodTop_Base
  bone2Name=head
  swing1LimitDegrees=45
  swing1Motion=Free
  swing2LimitDegrees=45
  swing2Motion=Free
  twistLimitDegrees=45
  twistMotion=Free
 }
 {
  bone1Name=dyn_hoodSide_Base_r
  bone2Name=head
  swing1LimitDegrees=45
  swing1Motion=Free
  swing2LimitDegrees=45
  swing2Motion=Free
  twistLimitDegrees=45
  twistMotion=Free
 }
 {
  bone1Name=dyn_hoodSide_Base_l
  bone2Name=head
  swing1LimitDegrees=45
  swing1Motion=Free
  swing2LimitDegrees=45
  swing2Motion=Free
  twistLimitDegrees=45
  twistMotion=Free
 }
 {
  bone1Name=dyn_hoodBack_Base_r
  bone2Name=neck_02
  swing1LimitDegrees=45
  swing1Motion=Free
  swing2LimitDegrees=45
  swing2Motion=Free
  twistLimitDegrees=45
  twistMotion=Free
 }
 {
  bone1Name=dyn_hoodBack_Base
  bone2Name=neck_02
  swing1LimitDegrees=45
  swing1Motion=Free
  swing2LimitDegrees=45
  swing2Motion=Free
  twistLimitDegrees=45
  twistMotion=Free
 }
 {
  bone1Name=dyn_hoodBack_Base_l
  bone2Name=neck_02
  swing1LimitDegrees=45
  swing1Motion=Free
  swing2LimitDegrees=45
  swing2Motion=Free
  twistLimitDegrees=45
  twistMotion=Free
 }
 {bone1Name=dyn_hoodTop, bone2Name=dyn_hoodTop_Base, swing1LimitDegrees=25, swing2LimitDegrees=12, twistLimitDegrees=6}
 {bone1Name=dyn_hoodSide_r, bone2Name=dyn_hoodSide_Base_r, swing1LimitDegrees=12, swing2LimitDegrees=30, twistLimitDegrees=10}
 {bone1Name=dyn_hoodBack_r, bone2Name=dyn_hoodBack_Base_r, swing1LimitDegrees=45, swing2LimitDegrees=10, twistLimitDegrees=1}
 {bone1Name=dyn_hoodBack, bone2Name=dyn_hoodBack_Base, swing1LimitDegrees=45, swing2LimitDegrees=10, twistLimitDegrees=20}
 {bone1Name=dyn_hoodBack_l, bone2Name=dyn_hoodBack_Base_l, swing1LimitDegrees=45, swing2LimitDegrees=10, twistLimitDegrees=1}
 {bone1Name=dyn_hoodSide_l, bone2Name=dyn_hoodSide_Base_l, swing1LimitDegrees=15, swing2LimitDegrees=12, twistLimitDegrees=10}
 {
  bone1Name=neck_02
  bone2Name=spine_05
  swing1LimitDegrees=45
  swing1Motion=Free
  swing2LimitDegrees=45
  swing2Motion=Free
  twistLimitDegrees=45
  twistMotion=Free
 }
 {
  bone1Name=clavicle_l
  bone2Name=spine_05
  swing1LimitDegrees=45
  swing1Motion=Free
  swing2LimitDegrees=45
  swing2Motion=Free
  twistLimitDegrees=45
  twistMotion=Free
 }
 {
  bone1Name=clavicle_r
  bone2Name=spine_05
  swing1LimitDegrees=45
  swing1Motion=Free
  swing2LimitDegrees=45
  swing2Motion=Free
  twistLimitDegrees=45
  twistMotion=Free
 }
]
