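# Animation Blueprint for the Sahara Cloak FaceAcc character part, adding procedural cape and hood motion to the copied body pose.
# Clavicle tweaks, a deform Control Rig (off while skydiving), weapon- and arm-driven offsets and head RigidBody feed a cached pose.
# `bRigidBodyEnabled` selects backbling-dependent RigidBody with crouch collision offset, or DBNO-dependent LookAt/Trail chains.
# The EventGraph updates arm alphas, crouch and gravity, and resets dynamics around skydiving.
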AnimBlueprint /BRCosmetics/Characters/Player/Male/Medium/Bodies/M_MED_Sahara_Cloak/Meshes/Parts/M_MED_Sahara_Cloak_FaceAcc_AnimBP
target_skeleton=&/M_MED_Sahara_Cloak_FaceAcc_Skeleton
graphs [
 AnimGraph {nodes [
   n1 = "cape_03_r.translateY = upperarm_r.rotateX * -0.3 - Bone Driven Controller"!(ComponentPose:Component Space Pose Link Structure, bAlphaBoolEnabled:bool=True) -> (Pose:Component Space Pose Link Structure) {
    is=title AnimGraphNode_BoneDrivenController
    alphaInputType=Bool
    bAffectTargetTranslationY=true
    multiplier=-0.3
    sourceComponent=RotationX
    alphaBoolBlend {blendOption=Linear}
    sourceBone {boneName=upperarm_r}
    targetBone {boneName=cape_03_r}
   }
   "cape_03_l.translateY = upperarm_l.rotateX * -0.6 - Bone Driven Controller"!(ComponentPose:Component Space Pose Link Structure=@n1, bAlphaBoolEnabled:bool=True) -> (Pose:Component Space Pose Link Structure) {
    is=title AnimGraphNode_BoneDrivenController
    alphaInputType=Bool
    bAffectTargetTranslationY=true
    multiplier=-0.6
    sourceComponent=RotationX
    alphaBoolBlend {blendOption=Linear}
    sourceBone {boneName=upperarm_l}
    targetBone {boneName=cape_03_l}
   }
   n2 = Look At - Bone: cape_mid_02_l(ComponentPose=Local To Component(LocalPose=Use cached pose 'initialPoseCache'()), Alpha=1) {
    alphaBoolBlend {blendOption=Linear}
    boneToModify {boneName=cape_mid_02_l}
    lookAtLocation=<0 -40 -11>
    lookAtTarget {boneReference {boneName=pelvis}}
    lookAt_Axis {axis=<1 0>}
   }
   n3 = Look At - Bone: cape_mid_02_r(ComponentPose=@n2, Alpha=1) {
    alphaBoolBlend {blendOption=Linear}
    boneToModify {boneName=cape_mid_02_r}
    lookAtLocation=<0 -40 11>
    lookAtTarget {boneReference {boneName=pelvis}}
    lookAt_Axis {axis=<1 0>}
   }
   n4 = Look At - Bone: cape_02_l(ComponentPose=@n3, Alpha=1) {
    alphaBoolBlend {blendOption=Linear}
    boneToModify {boneName=cape_02_l}
    lookAtLocation=<0 -25 -21>
    lookAtTarget {boneReference {boneName=pelvis}}
    lookAt_Axis {axis=<1 0>}
   }
   n5 = Look At - Bone: cape_02_r(ComponentPose=@n4, Alpha=1) {
    alphaBoolBlend {blendOption=Linear}
    boneToModify {boneName=cape_02_r}
    lookAtLocation=<0 -25 21>
    lookAtTarget {boneReference {boneName=pelvis}}
    lookAt_Axis {axis=<1 0>}
   }
   n6 = Look At - Bone: cape_mid_05_l(ComponentPose=@n5, Alpha=1) {
    alphaBoolBlend {blendOption=Linear}
    boneToModify {boneName=cape_mid_05_l}
    lookAtLocation=<-75 -15 -11>
    lookAtTarget {boneReference {boneName=pelvis}}
    lookAt_Axis {axis=<1 0>}
   }
   n7 = Look At - Bone: cape_mid_05_r(ComponentPose=@n6, Alpha=1) {
    alphaBoolBlend {blendOption=Linear}
    boneToModify {boneName=cape_mid_05_r}
    lookAtLocation=<-75 -15 11>
    lookAtTarget {boneReference {boneName=pelvis}}
    lookAt_Axis {axis=<1 0>}
   }
   n8 = Look At - Bone: cape_06_l(ComponentPose=@n7, Alpha=1) {
    alphaBoolBlend {blendOption=Linear}
    boneToModify {boneName=cape_06_l}
    lookAtLocation=<-75 -8 -21>
    lookAtTarget {boneReference {boneName=pelvis}}
    lookAt_Axis {axis=<1 0>}
   }
   n9 = Look At - Bone: cape_06_r(ComponentPose=@n8, Alpha=1) {
    alphaBoolBlend {blendOption=Linear}
    boneToModify {boneName=cape_06_r}
    lookAtLocation=<-75 -8 21>
    lookAtTarget {boneReference {boneName=pelvis}}
    lookAt_Axis {axis=<1 0>}
   }
   n10 = Trail controller - Bone: cape_mid_07_l!(ComponentPose=@n9, Alpha=1, RelaxationSpeedScale:float=2, FakeVelocity:Vector="0,0,0") -> (Pose) {
    bActorSpaceFakeVel=true
    bShowTrailLocation=true
    bUsePlanarLimit=true
    chainLength=6
    debugLifeTime=0.1
    maxDeltaTime=0.0333
    relaxationSpeedScale=1.2
    alphaScaleBiasClamp {bInterpResult=true, interpSpeedDecreasing=25, interpSpeedIncreasing=25}
    baseJoint {boneName=root}
    rotationLimits [
     {limitMax=<180 180 180>}
     {limitMax=<180 180 180>}
     {limitMax=<180 180 180>}
     {limitMax=<180 180 180>}
     {limitMax=<180 180 180>}
     {limitMax=<180 180 180>}
    ]
    rotationOffsets [<0 0 0>, <0 0 0>, <0 0 0>, <0 0 0>, <0 0 0>, <0 0 0>]
    trailBone {boneName=cape_mid_07_l}
    trailRelaxationSpeed {editorCurveData {keys [<at 0 15 RCIM_Linear>,<at 1 6 RCIM_Linear>]}}
   }
   n11 = Trail controller - Bone: cape_mid_07_r!(ComponentPose=@n10, Alpha=1, RelaxationSpeedScale:float=2, FakeVelocity:Vector="0,0,0") -> (Pose) {
    bActorSpaceFakeVel=true
    bShowTrailLocation=true
    bUsePlanarLimit=true
    chainLength=6
    debugLifeTime=0.1
    maxDeltaTime=0.0333
    relaxationSpeedScale=1.2
    alphaScaleBiasClamp {bInterpResult=true, interpSpeedDecreasing=25, interpSpeedIncreasing=25}
    baseJoint {boneName=root}
    rotationLimits [
     {limitMax=<180 180 180>}
     {limitMax=<180 180 180>}
     {limitMax=<180 180 180>}
     {limitMax=<180 180 180>}
     {limitMax=<180 180 180>}
     {limitMax=<180 180 180>}
    ]
    rotationOffsets [<0 0 0>, <0 0 0>, <0 0 0>, <0 0 0>, <0 0 0>, <0 0 0>]
    trailBone {boneName=cape_mid_07_r}
    trailRelaxationSpeed {editorCurveData {keys [<at 0 15 RCIM_Linear>,<at 1 6 RCIM_Linear>]}}
   }
   n12 = Trail controller - Bone: cape_08_l!(ComponentPose=@n11, Alpha=1, RelaxationSpeedScale:float=2, FakeVelocity:Vector="0,0,0") -> (Pose) {
    bActorSpaceFakeVel=true
    bShowTrailLocation=true
    bUsePlanarLimit=true
    chainLength=6
    debugLifeTime=0.1
    maxDeltaTime=0.0333
    relaxationSpeedScale=1.2
    alphaScaleBiasClamp {bInterpResult=true, interpSpeedDecreasing=25, interpSpeedIncreasing=25}
    baseJoint {boneName=root}
    rotationLimits [
     {limitMax=<180 180 180>}
     {limitMax=<180 180 180>}
     {limitMax=<180 180 180>}
     {limitMax=<180 180 180>}
     {limitMax=<180 180 180>}
     {limitMax=<180 180 180>}
    ]
    rotationOffsets [<0 0 0>, <0 0 0>, <0 0 0>, <0 0 0>, <0 0 0>, <0 0 0>]
    trailBone {boneName=cape_08_l}
    trailRelaxationSpeed {editorCurveData {keys [<at 0 15 RCIM_Linear>,<at 1 6 RCIM_Linear>]}}
   }
   n13 = Trail controller - Bone: cape_08_r!(ComponentPose=@n12, Alpha=1, RelaxationSpeedScale:float=2, FakeVelocity:Vector="0,0,0") -> (Pose) {
    bActorSpaceFakeVel=true
    bShowTrailLocation=true
    bUsePlanarLimit=true
    chainLength=6
    debugLifeTime=0.1
    maxDeltaTime=0.0333
    relaxationSpeedScale=1.2
    alphaScaleBiasClamp {bInterpResult=true, interpSpeedDecreasing=25, interpSpeedIncreasing=25}
    baseJoint {boneName=root}
    rotationLimits [
     {limitMax=<180 180 180>}
     {limitMax=<180 180 180>}
     {limitMax=<180 180 180>}
     {limitMax=<180 180 180>}
     {limitMax=<180 180 180>}
     {limitMax=<180 180 180>}
    ]
    rotationOffsets [<0 0 0>, <0 0 0>, <0 0 0>, <0 0 0>, <0 0 0>, <0 0 0>]
    trailBone {boneName=cape_08_r}
    trailRelaxationSpeed {editorCurveData {keys [<at 0 15 RCIM_Linear>,<at 1 6 RCIM_Linear>]}}
   }
   n14 = Component To Local(ComponentPose=@n13)
   n15 = Look At - Bone: cape_02_l(ComponentPose=Local To Component(LocalPose=Use cached pose 'initialPoseCache'()), Alpha=1) {
    alphaBoolBlend {blendOption=Linear}
    boneToModify {boneName=cape_02_l}
    lookAtLocation=<0 -30 -21>
    lookAtTarget {boneReference {boneName=pelvis}}
    lookAt_Axis {axis=<1 0>}
   }
   n16 = Look At - Bone: cape_02_r(ComponentPose=@n15, Alpha=1) {
    alphaBoolBlend {blendOption=Linear}
    boneToModify {boneName=cape_02_r}
    lookAtLocation=<0 -30 21>
    lookAtTarget {boneReference {boneName=pelvis}}
    lookAt_Axis {axis=<1 0>}
   }
   n17 = Get CapeTrailRelaxSpeedScale() -> (CapeTrailRelaxSpeedScale:float)
   n18 = Trail controller - Bone: cape_mid_07_l(
    ComponentPose=@n16
    Alpha=1
    RelaxationSpeedScale=@n17
    FakeVelocity=Get CapeTrailFakeVelocityMidLeft_b() -> (CapeTrailFakeVelocityMidLeft_b:Vector)
   ) {
    bActorSpaceFakeVel=true
    bShowTrailLocation=true
    bUsePlanarLimit=true
    chainLength=6
    debugLifeTime=0.1
    maxDeltaTime=0.0333
    relaxationSpeedScale=1.2
    alphaScaleBiasClamp {bInterpResult=true, interpSpeedDecreasing=25, interpSpeedIncreasing=25}
    baseJoint {boneName=root}
    planarLimits [
     {drivingBone {boneName=root}, planeTransform {location=<0 0 10>, rotation=<rot 0 0 0>, scale=<1 1 1>}}
     {drivingBone {boneName=spine_03}, planeTransform {location=<0 -20 0>, rotation=<rot 0 15 -89.9999>, scale=<1 1 1>}}
    ]
    rotationLimits [
     {limitMax=<180 180 180>}
     {limitMax=<180 180 180>}
     {limitMax=<180 180 180>}
     {limitMax=<180 180 180>}
     {limitMax=<180 180 180>}
     {limitMax=<180 180 180>}
    ]
    rotationOffsets [<0 0 0>, <0 0 0>, <0 0 0>, <0 0 0>, <0 0 0>, <0 0 0>]
    trailBone {boneName=cape_mid_07_l}
    trailRelaxationSpeed {editorCurveData {keys [<at 0 15 RCIM_Linear>,<at 0.5289 13.1838 RCIM_Linear>,<at 1 6 RCIM_Linear>]}}
   }
   n19 = Trail controller - Bone: cape_mid_07_r(
    ComponentPose=@n18
    Alpha=1
    RelaxationSpeedScale=@n17
    FakeVelocity=Get CapeTrailFakeVelocityMidRight_b() -> (CapeTrailFakeVelocityMidRight_b:Vector)
   ) {
    bActorSpaceFakeVel=true
    bShowTrailLocation=true
    bUsePlanarLimit=true
    chainLength=6
    debugLifeTime=0.1
    maxDeltaTime=0.0333
    relaxationSpeedScale=1.2
    alphaScaleBiasClamp {bInterpResult=true, interpSpeedDecreasing=25, interpSpeedIncreasing=25}
    baseJoint {boneName=root}
    planarLimits [
     {drivingBone {boneName=root}, planeTransform {location=<0 0 10>, rotation=<rot 0 0 0>, scale=<1 1 1>}}
     {drivingBone {boneName=spine_03}, planeTransform {location=<0 -20 0>, rotation=<rot 0 15 -89.9999>, scale=<1 1 1>}}
    ]
    rotationLimits [
     {limitMax=<180 180 180>}
     {limitMax=<180 180 180>}
     {limitMax=<180 180 180>}
     {limitMax=<180 180 180>}
     {limitMax=<180 180 180>}
     {limitMax=<180 180 180>}
    ]
    rotationOffsets [<0 0 0>, <0 0 0>, <0 0 0>, <0 0 0>, <0 0 0>, <0 0 0>]
    trailBone {boneName=cape_mid_07_r}
    trailRelaxationSpeed {editorCurveData {keys [<at 0 15 RCIM_Linear>,<at 0.5289 13.1838 RCIM_Linear>,<at 1 6 RCIM_Linear>]}}
   }
   n20 = Get CapeTrailRelaxSpeedScale()
   n21 = Trail controller - Bone: cape_08_l(
    ComponentPose=@n19
    Alpha=1
    RelaxationSpeedScale=@n20
    FakeVelocity=Get CapeTrailFakeVelocityLowLeft_b() -> (CapeTrailFakeVelocityLowLeft_b:Vector)
   ) {
    bActorSpaceFakeVel=true
    bShowTrailLocation=true
    bUsePlanarLimit=true
    chainLength=6
    debugLifeTime=0.1
    maxDeltaTime=0.0333
    relaxationSpeedScale=1.2
    alphaScaleBiasClamp {bInterpResult=true, interpSpeedDecreasing=25, interpSpeedIncreasing=25}
    baseJoint {boneName=root}
    planarLimits [
     {drivingBone {boneName=root}, planeTransform {location=<0 0 10>, rotation=<rot 0 0 0>, scale=<1 1 1>}}
     {drivingBone {boneName=spine_03}, planeTransform {location=<0 -20 0>, rotation=<rot 0 15 -89.9999>, scale=<1 1 1>}}
    ]
    rotationLimits [
     {limitMax=<180 180 180>}
     {limitMax=<180 180 180>}
     {limitMax=<180 180 180>}
     {limitMax=<180 180 180>}
     {limitMax=<180 180 180>}
     {limitMax=<180 180 180>}
    ]
    rotationOffsets [<0 0 0>, <0 0 0>, <0 0 0>, <0 0 0>, <0 0 0>, <0 0 0>]
    trailBone {boneName=cape_08_l}
    trailRelaxationSpeed {editorCurveData {keys [<at 0 15 RCIM_Linear>,<at 0.5289 13.1838 RCIM_Linear>,<at 1 6 RCIM_Linear>]}}
   }
   n22 = Trail controller - Bone: cape_08_r(
    ComponentPose=@n21
    Alpha=1
    RelaxationSpeedScale=@n20
    FakeVelocity=Get CapeTrailFakeVelocityLowRight_b() -> (CapeTrailFakeVelocityLowRight_b:Vector)
   ) {
    bActorSpaceFakeVel=true
    bShowTrailLocation=true
    bUsePlanarLimit=true
    chainLength=6
    debugLifeTime=0.1
    maxDeltaTime=0.0333
    relaxationSpeedScale=1.2
    alphaScaleBiasClamp {bInterpResult=true, interpSpeedDecreasing=25, interpSpeedIncreasing=25}
    baseJoint {boneName=root}
    planarLimits [
     {drivingBone {boneName=root}, planeTransform {location=<0 0 10>, rotation=<rot 0 0 0>, scale=<1 1 1>}}
     {drivingBone {boneName=spine_03}, planeTransform {location=<0 -20 0>, rotation=<rot 0 15 -89.9999>, scale=<1 1 1>}}
    ]
    rotationLimits [
     {limitMax=<180 180 180>}
     {limitMax=<180 180 180>}
     {limitMax=<180 180 180>}
     {limitMax=<180 180 180>}
     {limitMax=<180 180 180>}
     {limitMax=<180 180 180>}
    ]
    rotationOffsets [<0 0 0>, <0 0 0>, <0 0 0>, <0 0 0>, <0 0 0>, <0 0 0>]
    trailBone {boneName=cape_08_r}
    trailRelaxationSpeed {editorCurveData {keys [<at 0 15 RCIM_Linear>,<at 0.5289 13.1838 RCIM_Linear>,<at 1 6 RCIM_Linear>]}}
   }
   n23 = "Transform (Modify) Bone - Bone: cape_01_l"(ComponentPose=@n22, Alpha=1, Translation="0,-1,0", Rotation="0,0,0", Scale="1,1,1") {translationMode=BMM_Additive, alphaBoolBlend {blendOption=Linear}, boneToModify {boneName=cape_01_l}, translation {y=-1}}
   n24 = Component To Local(ComponentPose=@n23)
   n25 = Get bIsDBNO()
   n26 = Get crouch alpha inv() -> (crouch alpha inv:double)
   n27 = "Transform (Modify) Bone - Bone: dyn_crouch_collision"(
    ComponentPose=Local To Component(LocalPose=Use cached pose 'initialPoseCache'())
    Alpha=@n26
    Translation="0.000,0.000,-49.249"
    Rotation="0,0,0"
    Scale="1,1,1"
   ) {
    translationMode=BMM_Additive
    translationSpace=BCS_BoneSpace
    alphaBoolBlend {blendOption=Linear}
    alphaScaleBiasClamp {bInterpResult=true}
    boneToModify {boneName=dyn_crouch_collision}
    translation {z=-49.249}
   }
   n28 = RigidBody!(ComponentPose=@n27, Alpha, OverrideWorldGravity:Vector="0,0,0", ExternalForce) -> (Pose) {
    bClampLinearTranslationLimitToRefPose=true
    bOverrideWorldGravity=true
    overridePhysicsAsset=&/M_MED_Sahara_CloakBackpack_FaceAcc_PhysicsAsset
    simulationSpace=BaseBoneSpace
    alphaBoolBlend {blendOption=Linear}
    baseBoneRef {boneName=root}
    componentAppliedLinearAccClamp=<3000 3000 3000>
    componentLinearAccScale=<0.7 0.5 0.5>
    componentLinearVelScale=<0.7 0.5 0.5>
    overrideWorldGravity {z=-980}
   }
   n29 = Component To Local(ComponentPose=@n28)
   n30 = "Transform (Modify) Bone - Bone: dyn_crouch_collision"(
    ComponentPose=Local To Component(LocalPose=Use cached pose 'initialPoseCache'())
    Alpha=@n26
    Translation="0.000,0.000,-49.249"
    Rotation="0,0,0"
    Scale="1,1,1"
   ) {
    translationMode=BMM_Additive
    translationSpace=BCS_BoneSpace
    alphaBoolBlend {blendOption=Linear}
    alphaScaleBiasClamp {bInterpResult=true}
    boneToModify {boneName=dyn_crouch_collision}
    translation {z=-49.249}
   }
   n31 = RigidBody(ComponentPose=@n30, OverrideWorldGravity=Get CapeGravityOverride() -> (CapeGravityOverride:Vector), ExternalForce="0,0,20") {
    bClampLinearTranslationLimitToRefPose=true
    bOverrideWorldGravity=true
    overridePhysicsAsset=&/M_MED_Sahara_CloakHood_FaceAcc_PhysicsAsset
    simulationSpace=BaseBoneSpace
    alphaBoolBlend {blendOption=Linear}
    baseBoneRef {boneName=root}
    componentAppliedLinearAccClamp=<3000 3000 3000>
    componentLinearAccScale=<0.7 0.5 0.5>
    componentLinearVelScale=<0.7 0.5 0.5>
    overrideWorldGravity {z=-600}
   }
   n32 = Component To Local(ComponentPose=@n31)
   n33 = Get isWearingBackbling()
   n34 = Control Rig!(
    Source=Blend Poses by bool(
     BlendPose_0=Blend Poses by bool(BlendPose_0=@n29, BlendPose_1=@n32, bActiveValue=@n33) {blendType=Linear}
     BlendPose_1=Blend Poses by bool(BlendPose_0=@n14, BlendPose_1=@n24, bActiveValue=@n25) {blendType=Linear}
     bActiveValue=Get bRigidBodyEnabled()
    ) {blendType=Linear}
    Alpha=0
    IsSkydiving:double=Get SkydiveFloat() -> (SkydiveFloat:float)
   ) -> (Pose) {
    alphaBoolBlend {blendOption=Linear}
    controlRigAssetReference {blueprintRigClass=&/M_MED_Sahara_Cloak_FaceAcc_CtrlRig.M_MED_Sahara_Cloak_FaceAcc_CtrlRig_C}
   }
   Output Pose(Result=@n34) {layerGroup=None}
   n35 = "Transform (Modify) Bone - Bone: clavicle_l"!(
    ComponentPose=Local To Component(LocalPose=Copy Pose From Mesh() {bCopyCurves=true, bUseAttachedParent=true})
    bAlphaBoolEnabled:bool=True
    Translation="0,0,5"
    Rotation="0,0,0"
    Scale="1,1,1"
   ) -> (Pose) {
    alphaInputType=Bool
    rotationMode=BMM_Additive
    rotationSpace=BCS_BoneSpace
    scaleSpace=BCS_BoneSpace
    translationMode=BMM_Additive
    translationSpace=BCS_BoneSpace
    alphaBoolBlend {blendInTime=0.1, blendOption=Linear, blendOutTime=0.1}
    alphaScaleBiasClamp {bInterpResult=true}
    boneToModify {boneName=clavicle_l}
    translation {z=5}
   }
   n36 = "Transform (Modify) Bone - Bone: clavicle_r"!(ComponentPose=@n35, bAlphaBoolEnabled:bool=True, Translation="-5,0,-3", Rotation="0,0,0", Scale="1,1,1") -> (Pose) {
    alphaInputType=Bool
    rotationMode=BMM_Additive
    rotationSpace=BCS_BoneSpace
    scaleMode=BMM_Additive
    scaleSpace=BCS_BoneSpace
    translationMode=BMM_Additive
    translationSpace=BCS_BoneSpace
    alphaBoolBlend {blendInTime=0.1, blendOption=Linear, blendOutTime=0.1}
    alphaScaleBiasClamp {bInterpResult=true}
    boneToModify {boneName=clavicle_r}
    translation {x=-5, z=-3}
   }
   n37 = "Transform (Modify) Bone - Bone: clavicle_l"(ComponentPose=@n36, bAlphaBoolEnabled=True, Translation="0,5,0", Rotation="20,-10,0", Scale="1,1,1") {
    alphaInputType=Bool
    rotationMode=BMM_Additive
    rotationSpace=BCS_BoneSpace
    scaleSpace=BCS_BoneSpace
    translationMode=BMM_Additive
    translationSpace=BCS_BoneSpace
    alphaBoolBlend {blendInTime=0.1, blendOption=Linear, blendOutTime=0.1}
    alphaScaleBiasClamp {bInterpResult=true}
    boneToModify {boneName=clavicle_l}
    rotation {pitch=20, yaw=-10}
    translation {y=5}
   }
   n38 = "Transform (Modify) Bone - Bone: clavicle_r"(ComponentPose=@n37, bAlphaBoolEnabled=True, Translation="-3,-5,0", Rotation="10,-10,0", Scale="1,1,1") {
    alphaInputType=Bool
    rotationMode=BMM_Additive
    rotationSpace=BCS_BoneSpace
    scaleMode=BMM_Additive
    scaleSpace=BCS_BoneSpace
    translationMode=BMM_Additive
    translationSpace=BCS_BoneSpace
    alphaBoolBlend {blendInTime=0.1, blendOption=Linear, blendOutTime=0.1}
    alphaScaleBiasClamp {bInterpResult=true}
    boneToModify {boneName=clavicle_r}
    rotation {pitch=10, yaw=-10}
    translation=<-3 -5>
   }
   n39 = "Transform (Modify) Bone - Bone: clavicle_l"(ComponentPose=@n38, bAlphaBoolEnabled=True, Translation="-5,-2,0", Rotation="0,-15,0", Scale="1,1,1") {
    alphaInputType=Bool
    rotationMode=BMM_Additive
    rotationSpace=BCS_BoneSpace
    scaleSpace=BCS_BoneSpace
    translationMode=BMM_Additive
    translationSpace=BCS_BoneSpace
    alphaBoolBlend {blendInTime=0.1, blendOption=Linear, blendOutTime=0.1}
    alphaScaleBiasClamp {bInterpResult=true}
    boneToModify {boneName=clavicle_l}
    rotation {yaw=-15}
    translation=<-5 -2>
   }
   n40 = "Transform (Modify) Bone - Bone: clavicle_r"(ComponentPose=@n39, bAlphaBoolEnabled=True, Translation="5,2,0", Rotation="0,-15,0", Scale="1,1,1") {
    alphaInputType=Bool
    rotationMode=BMM_Additive
    rotationSpace=BCS_BoneSpace
    scaleMode=BMM_Additive
    scaleSpace=BCS_BoneSpace
    translationMode=BMM_Additive
    translationSpace=BCS_BoneSpace
    alphaBoolBlend {blendInTime=0.1, blendOption=Linear, blendOutTime=0.1}
    alphaScaleBiasClamp {bInterpResult=true}
    boneToModify {boneName=clavicle_r}
    rotation {yaw=-15}
    translation=<5 2>
   }
   n41 = "Transform (Modify) Bone - Bone: clavicle_l"(
    ComponentPose=Local To Component(LocalPose=Copy Pose From Mesh() {bCopyCurves=true, bUseAttachedParent=true})
    bAlphaBoolEnabled=True
    Translation="0,0,0"
    Rotation="10,0,0"
    Scale="1,1,1"
   ) {
    alphaInputType=Bool
    rotationMode=BMM_Additive
    rotationSpace=BCS_BoneSpace
    scaleSpace=BCS_BoneSpace
    translationMode=BMM_Additive
    translationSpace=BCS_BoneSpace
    alphaBoolBlend {blendInTime=0.1, blendOption=Linear, blendOutTime=0.1}
    alphaScaleBiasClamp {bInterpResult=true}
    boneToModify {boneName=clavicle_l}
    rotation {pitch=10}
   }
   n42 = "Transform (Modify) Bone - Bone: clavicle_r"(ComponentPose=@n41, bAlphaBoolEnabled=True, Translation="-5,0,-3", Rotation="0,0,0", Scale="1,1,1") {
    alphaInputType=Bool
    rotationMode=BMM_Additive
    rotationSpace=BCS_BoneSpace
    scaleMode=BMM_Additive
    scaleSpace=BCS_BoneSpace
    translationMode=BMM_Additive
    translationSpace=BCS_BoneSpace
    alphaBoolBlend {blendInTime=0.1, blendOption=Linear, blendOutTime=0.1}
    alphaScaleBiasClamp {bInterpResult=true}
    boneToModify {boneName=clavicle_r}
    translation {x=-5, z=-3}
   }
   n43 = "Transform (Modify) Bone - Bone: clavicle_l"(ComponentPose=@n42, bAlphaBoolEnabled=True, Translation="5,4,2", Rotation="4,-20,0", Scale="1,1,1") {
    alphaInputType=Bool
    rotationMode=BMM_Additive
    rotationSpace=BCS_BoneSpace
    scaleSpace=BCS_BoneSpace
    translationMode=BMM_Additive
    translationSpace=BCS_BoneSpace
    alphaBoolBlend {blendInTime=0.1, blendOption=Linear, blendOutTime=0.1}
    alphaScaleBiasClamp {bInterpResult=true}
    boneToModify {boneName=clavicle_l}
    rotation {pitch=4, yaw=-20}
    translation=<5 4 2>
   }
   n44 = "Transform (Modify) Bone - Bone: clavicle_r"(ComponentPose=@n43, bAlphaBoolEnabled=True, Translation="-8,-9,0", Rotation="10,-30,0", Scale="1,1,1") {
    alphaInputType=Bool
    rotationMode=BMM_Additive
    rotationSpace=BCS_BoneSpace
    scaleMode=BMM_Additive
    scaleSpace=BCS_BoneSpace
    translationMode=BMM_Additive
    translationSpace=BCS_BoneSpace
    alphaBoolBlend {blendInTime=0.1, blendOption=Linear, blendOutTime=0.1}
    alphaScaleBiasClamp {bInterpResult=true}
    boneToModify {boneName=clavicle_r}
    rotation {pitch=10, yaw=-30}
    translation=<-8 -9>
   }
   n45 = "Transform (Modify) Bone - Bone: clavicle_l"(ComponentPose=@n44, bAlphaBoolEnabled=True, Translation="-5,-2,0", Rotation="0,-15,0", Scale="1,1,1") {
    alphaInputType=Bool
    rotationMode=BMM_Additive
    rotationSpace=BCS_BoneSpace
    scaleSpace=BCS_BoneSpace
    translationMode=BMM_Additive
    translationSpace=BCS_BoneSpace
    alphaBoolBlend {blendInTime=0.1, blendOption=Linear, blendOutTime=0.1}
    alphaScaleBiasClamp {bInterpResult=true}
    boneToModify {boneName=clavicle_l}
    rotation {yaw=-15}
    translation=<-5 -2>
   }
   n46 = "Transform (Modify) Bone - Bone: clavicle_r"(ComponentPose=@n45, bAlphaBoolEnabled=True, Translation="5,2,0", Rotation="0,-15,0", Scale="1,1,1") {
    alphaInputType=Bool
    rotationMode=BMM_Additive
    rotationSpace=BCS_BoneSpace
    scaleMode=BMM_Additive
    scaleSpace=BCS_BoneSpace
    translationMode=BMM_Additive
    translationSpace=BCS_BoneSpace
    alphaBoolBlend {blendInTime=0.1, blendOption=Linear, blendOutTime=0.1}
    alphaScaleBiasClamp {bInterpResult=true}
    boneToModify {boneName=clavicle_r}
    rotation {yaw=-15}
    translation=<5 2>
   }
   n47 = Get ADS_UP_L_Pistol() -> (ADS_UP_L_Pistol:double)
   n48 = "Transform (Modify) Bone - Bone: dyn_capeTopShoulder_Base_l"(ComponentPose=@n46, Alpha=@n47, Translation="40,-50,20", Rotation="-0.000,-0.000,-11.257", Scale="1,1,1") {
    rotationMode=BMM_Additive
    rotationSpace=BCS_BoneSpace
    translationMode=BMM_Additive
    translationSpace=BCS_BoneSpace
    alphaBoolBlend {blendOption=Linear}
    alphaScaleBiasClamp {bInterpResult=true}
    boneToModify {boneName=dyn_capeTopShoulder_Base_l}
    rotation=<rot 0 0 -11.257>
    translation=<40 -50 20>
   }
   n49 = "Transform (Modify) Bone - Bone: dyn_capeFrontShoulder_Base_l"(ComponentPose=@n48, Alpha=@n47, Translation="40.005,-103.188,-2.532", Rotation="0,0,0", Scale="1,1,1") {
    rotationMode=BMM_Additive
    rotationSpace=BCS_BoneSpace
    translationMode=BMM_Additive
    translationSpace=BCS_BoneSpace
    alphaBoolBlend {blendOption=Linear}
    alphaScaleBiasClamp {bInterpResult=true}
    boneToModify {boneName=dyn_capeFrontShoulder_Base_l}
    translation=<40.005 -103.188 -2.532>
   }
   n50 = Get ADS_UP_R_Pistol() -> (ADS_UP_R_Pistol:double)
   n51 = "Transform (Modify) Bone - Bone: dyn_capeTopShoulder_Base_r"(ComponentPose=@n49, Alpha=@n50, Translation="0.903,2.469,2.088", Rotation="0.088,-19.952,73.127", Scale="1,1,1") {
    rotationMode=BMM_Additive
    rotationSpace=BCS_BoneSpace
    translationMode=BMM_Additive
    translationSpace=BCS_BoneSpace
    alphaBoolBlend {blendOption=Linear}
    alphaScaleBiasClamp {bInterpResult=true}
    boneToModify {boneName=dyn_capeTopShoulder_Base_r}
    rotation=<rot 0.088 -19.952 73.127>
    translation=<0.903 2.469 2.088>
   }
   n52 = "Transform (Modify) Bone - Bone: dyn_capeFrontShoulder_Base_r"(ComponentPose=@n51, Alpha=@n50, Translation="-0.754,-27.683,-9.987", Rotation="0,0,0", Scale="1,1,1") {
    rotationMode=BMM_Additive
    rotationSpace=BCS_BoneSpace
    translationMode=BMM_Additive
    translationSpace=BCS_BoneSpace
    alphaBoolBlend {blendOption=Linear}
    alphaScaleBiasClamp {bInterpResult=true}
    boneToModify {boneName=dyn_capeFrontShoulder_Base_r}
    translation=<-0.754 -27.683 -9.987>
   }
   n53 = "Transform (Modify) Bone - Bone: dyn_capeChainStart"(ComponentPose=@n52, Alpha=@n50, Translation="-0.000,-0.000,0.779", Rotation="-0.000,-0.000,-8.450", Scale="1,1,1") {
    rotationMode=BMM_Additive
    rotationSpace=BCS_BoneSpace
    translationMode=BMM_Additive
    translationSpace=BCS_BoneSpace
    alphaBoolBlend {blendOption=Linear}
    alphaScaleBiasClamp {bInterpResult=true}
    boneToModify {boneName=dyn_capeChainStart}
    rotation=<rot 0 0 -8.45>
    translation {z=0.779}
   }
   n54 = "Transform (Modify) Bone - Bone: dyn_capeFrontShoulder_Base_l"(ComponentPose=@n53, Alpha=Get ADS_UP_L_Assault() -> (ADS_UP_L_Assault:double), Translation="27,-80,-38", Rotation="0,0,0", Scale="1,1,1") {
    rotationMode=BMM_Additive
    rotationSpace=BCS_BoneSpace
    translationMode=BMM_Additive
    translationSpace=BCS_BoneSpace
    alphaBoolBlend {blendOption=Linear}
    alphaScaleBiasClamp {bInterpResult=true}
    boneToModify {boneName=dyn_capeFrontShoulder_Base_l}
    translation=<27 -80 -38>
   }
   n55 = Get ADS_UP_R_Assault() -> (ADS_UP_R_Assault:double)
   n56 = "Transform (Modify) Bone - Bone: dyn_capeFrontShoulder_Base_r"(ComponentPose=@n54, Alpha=@n55, Translation="-0.000,-22,-0.000", Rotation="0,0,0", Scale="1,1,1") {
    rotationMode=BMM_Additive
    rotationSpace=BCS_BoneSpace
    translationMode=BMM_Additive
    translationSpace=BCS_BoneSpace
    alphaBoolBlend {blendOption=Linear}
    alphaScaleBiasClamp {bInterpResult=true}
    boneToModify {boneName=dyn_capeFrontShoulder_Base_r}
    translation {y=-22}
   }
   n57 = "Transform (Modify) Bone - Bone: dyn_capeChainStart"(ComponentPose=@n56, Alpha=@n55, Translation="0,0,0", Rotation="0,0,0", Scale="1,1,1") {
    rotationMode=BMM_Additive
    rotationSpace=BCS_BoneSpace
    translationMode=BMM_Additive
    translationSpace=BCS_BoneSpace
    alphaBoolBlend {blendOption=Linear}
    alphaScaleBiasClamp {bInterpResult=true}
    boneToModify {boneName=dyn_capeChainStart}
   }
   n58 = Control Rig!(
    Source=Blend Poses by bool(
     BlendPose_0=Component To Local(ComponentPose=@n57)
     BlendPose_1=Component To Local(ComponentPose=@n40)
     bActiveValue=Get Alt_weapon() -> (Alt_weapon:bool)
    ) {blendType=Linear}
    Alpha=Get NoSkyDive() -> (NoSkyDive:float)
   ) -> (Pose) {
    alphaBoolBlend {blendOption=Linear}
    controlRigAssetReference {blueprintRigClass=&/M_MED_DeformRig_Sahara_Cape_FaceAcc_ControlRig.M_MED_DeformRig_Sahara_Cape_FaceAcc_ControlRig_C}
   }
   n59 = Get Hood_ADS_UP_L_Pistol() -> (Hood_ADS_UP_L_Pistol:double)
   n60 = "Transform (Modify) Bone - Bone: dyn_hoodBack_Base_l"(ComponentPose=Local To Component(LocalPose=@n58), Alpha=@n59, Translation="0,-2,0", Rotation="0,0,0", Scale="1,1,1") {
    rotationMode=BMM_Additive
    rotationSpace=BCS_BoneSpace
    translationMode=BMM_Additive
    translationSpace=BCS_BoneSpace
    alphaBoolBlend {blendOption=Linear}
    alphaScaleBiasClamp {bInterpResult=true}
    boneToModify {boneName=dyn_hoodBack_Base_l}
    translation {y=-2}
   }
   n61 = Get Hood_ADS_UP_R_Pistol() -> (Hood_ADS_UP_R_Pistol:double)
   n62 = "Transform (Modify) Bone - Bone: dyn_hoodBack_Base_r"(ComponentPose=@n60, Alpha=@n61, Translation="2,0,-2", Rotation="0,0,0", Scale="1,1,1") {
    rotationMode=BMM_Additive
    rotationSpace=BCS_BoneSpace
    translationMode=BMM_Replace
    translationSpace=BCS_BoneSpace
    alphaBoolBlend {blendOption=Linear}
    alphaScaleBiasClamp {bInterpResult=true}
    boneToModify {boneName=dyn_hoodBack_Base_r}
    translation {x=2, z=-2}
   }
   n63 = "Transform (Modify) Bone - Bone: dyn_hoodSide_Base_l"(ComponentPose=@n62, Alpha=@n59, Translation="0,0,-5", Rotation="0,0,0", Scale="1,1,1") {
    rotationMode=BMM_Additive
    rotationSpace=BCS_BoneSpace
    translationMode=BMM_Replace
    translationSpace=BCS_BoneSpace
    alphaBoolBlend {blendOption=Linear}
    alphaScaleBiasClamp {bInterpResult=true}
    boneToModify {boneName=dyn_hoodSide_Base_l}
    translation {z=-5}
   }
   n64 = "Transform (Modify) Bone - Bone: dyn_hoodSide_Base_r"(ComponentPose=@n63, Alpha=@n61, Translation="-1.018,6.445,-11.479", Rotation="0,0,0", Scale="1,1,1") {
    rotationMode=BMM_Additive
    rotationSpace=BCS_BoneSpace
    translationMode=BMM_Additive
    translationSpace=BCS_BoneSpace
    alphaBoolBlend {blendOption=Linear}
    alphaScaleBiasClamp {bInterpResult=true}
    boneToModify {boneName=dyn_hoodSide_Base_r}
    translation=<-1.018 6.445 -11.479>
   }
   n65 = Get Hood_ADS_UP_L_Assault() -> (Hood_ADS_UP_L_Assault:double)
   n66 = "Transform (Modify) Bone - Bone: dyn_hoodBack_Base_l"(ComponentPose=@n64, Alpha=@n65, Translation="0,-2,0", Rotation="0,0,0", Scale="1,1,1") {
    rotationMode=BMM_Additive
    rotationSpace=BCS_BoneSpace
    translationMode=BMM_Additive
    translationSpace=BCS_BoneSpace
    alphaBoolBlend {blendOption=Linear}
    alphaScaleBiasClamp {bInterpResult=true}
    boneToModify {boneName=dyn_hoodBack_Base_l}
    translation {y=-2}
   }
   n67 = "Transform (Modify) Bone - Bone: dyn_hoodBack_Base_r"(ComponentPose=@n66, Alpha=@n65, Translation="0,-20,0", Rotation="0,0,0", Scale="1,1,1") {
    rotationMode=BMM_Additive
    rotationSpace=BCS_BoneSpace
    translationMode=BMM_Replace
    translationSpace=BCS_BoneSpace
    alphaBoolBlend {blendOption=Linear}
    alphaScaleBiasClamp {bInterpResult=true}
    boneToModify {boneName=dyn_hoodBack_Base_r}
    translation {y=-20}
   }
   n68 = "Transform (Modify) Bone - Bone: dyn_hoodSide_Base_l"(ComponentPose=@n67, Alpha=@n65, Translation="0,0,-5", Rotation="0,0,0", Scale="1,1,1") {
    rotationMode=BMM_Additive
    rotationSpace=BCS_BoneSpace
    translationMode=BMM_Replace
    translationSpace=BCS_BoneSpace
    alphaBoolBlend {blendOption=Linear}
    alphaScaleBiasClamp {bInterpResult=true}
    boneToModify {boneName=dyn_hoodSide_Base_l}
    translation {z=-5}
   }
   n69 = "Transform (Modify) Bone - Bone: dyn_hoodSide_Base_r"(ComponentPose=@n68, Alpha=@n65, Translation="0,5,-14", Rotation="0,0,0", Scale="1,1,1") {
    rotationMode=BMM_Additive
    rotationSpace=BCS_BoneSpace
    translationMode=BMM_Replace
    translationSpace=BCS_BoneSpace
    alphaBoolBlend {blendOption=Linear}
    alphaScaleBiasClamp {bInterpResult=true}
    boneToModify {boneName=dyn_hoodSide_Base_r}
    translation {y=5, z=-14}
   }
   n70 = "Constraint - Dyn_Head (RR)"!(ComponentPose=@n69, Alpha=1, ConstraintWeights_0=0.8, ConstraintWeights_1:float=0.2) -> (Pose) {
    alphaBoolBlend {blendOption=Linear}
    boneToModify {boneName=Dyn_Head}
    constraintSetup [
     {offsetOption=None, transformType=Rotation, targetBone {boneName=head}}
     {offsetOption=None, transformType=Rotation, targetBone {boneName=neck_02}}
    ]
    constraintWeights [1,1]
   }
   n71 = RigidBody!(ComponentPose=@n70, Alpha, ExternalForce) -> (Pose) {
    bClampLinearTranslationLimitToRefPose=true
    bOverrideWorldGravity=true
    overridePhysicsAsset=&/M_MED_Sahara_Cloak_FaceAcc_Physics
    simulationSpace=BaseBoneSpace
    alphaBoolBlend {blendOption=Linear}
    baseBoneRef {boneName=head}
    componentAppliedLinearAccClamp=<3000 3000 3000>
    componentLinearAccScale=<0.5 0.5 0.45>
    componentLinearVelScale=<0.5 0.55 0.55>
    overrideWorldGravity {z=-980}
   }
   Save cached pose 'initialPoseCache'(Pose=Component To Local(ComponentPose=@n71)) {cacheName=initialPoseCache}
  ]}
 EventGraph {nodes [
   Event OnBeginSkydiving()
   > Delay(Duration=0.1)
   > ResetDynamics(InTeleportType=TeleportPhysics)
   > ResetDynamics(InTeleportType=ResetPhysics)
   > Set The_SkydiveFloat(The_SkydiveFloat:float=1) -> (Output_Get:float)
   > Set NoSkyDive(NoSkyDive:float=0) -> (Output_Get:float)
   Event OnCharacterCustomizationLoaded()
   > Do Once()
   n1 = CheckForBackbling()
   > Set isWearingBackbling(isWearingBackbling=OR Boolean(A=@n1, B=@n1.bIsBackblingLong))
   Event OnEndCrouching()
   > Set crouch alpha(crouch alpha:double=0) -> (Output_Get:double)
   > Set crouch alpha inv(crouch alpha inv:double=1) -> (Output_Get:double)
   Event OnBeginCrouching()
   > Set crouch alpha(crouch alpha=1)
   > Set crouch alpha inv(crouch alpha inv=0)
   Event OnEndSkydiving()
   > Delay(Duration=0.1)
   > ResetDynamics(InTeleportType=TeleportPhysics)
   > ResetDynamics(InTeleportType=ResetPhysics)
   > Set The_SkydiveFloat(The_SkydiveFloat=0)
   > Set NoSkyDive(NoSkyDive=1)
   Event BlueprintUpdateAnimation()
   > n2 = Sequence!(execute) -> (then_0, then_1, then_2:Exec)
   Set Skel Mesh(execute=@n2.then_1, Skel Mesh=GetOwnerSkeletalMeshComponent())
   > n3 = Branch(Condition=Get bIsSkyDiving())
   n4 = Cast To FortPlayerAnimInstance(execute=@n3.else, Object=GetAnimInstance(self=GetOwnerSkeletalMeshComponent()))
   > Set Alt_weapon(Alt_weapon:bool=false) -> (Output_Get:bool)
   > n5 = Switch on EFortWeaponCoreAnimation!(execute, Selection:EFortWeaponCoreAnimation Enum=Get WeaponCoreAnim(self=@n4)) -> (
    Melee:Exec
    Pistol:Exec
    Shotgun:Exec
    PaperBlueprint:Exec
    Rifle:Exec
    MeleeOneHand:Exec
    MachinePistol:Exec
    RocketLauncher:Exec
    GrenadeLauncher:Exec
    GoingCommando:Exec
    AssaultRifle:Exec
    TacticalShotgun:Exec
    SniperRifle:Exec
    TrapPlacement:Exec
    ShoulderLauncher:Exec
    AbilityDecoTool:Exec
    Crossbow:Exec
    C4:Exec
    RemoteControl:Exec
    DualWield:Exec
    AR_BullPup:Exec
    AR_ForwardGrip:Exec
    MedPackPaddles:Exec
    SMG_P90:Exec
    AR_DrumGun:Exec
    Consumable_Small:Exec
    Consumable_Large:Exec
    Balloon:Exec
    MountedTurret:Exec
    CreativeTool:Exec
    ExplosiveBow:Exec
    AshtonIndigo:Exec
    AshtonChicago:Exec
    MeleeDualWield:Exec
    Unarmed:Exec
   )
   Branch(execute=[@n5.AssaultRifle, @n5.Rifle])
   > Set Alt_weapon(Alt_weapon=true)
   > n6 = Sequence!(execute) -> (then_0, then_1, then_2, then_3:Exec)
   n7 = Get Skel Mesh()
   > Set ADS_UP_L_Assault(ADS_UP_L_Assault:double=GetAlphaFromReferenceTransformDelta(SkeletalMeshComponent=@n7, BoneName=upperarm_l, BaseName=clavicle_l, InputRange=(X=10,Y=50), TwistAxis="0,0,1")) -> (Output_Get:double)
   > Set ADS_UP_L_Pistol(ADS_UP_L_Pistol:double=0) -> (Output_Get:double)
   Set ADS_UP_R_Assault(
    execute=@n6.then_1
    ADS_UP_R_Assault:double=GetAlphaFromReferenceTransformDelta(SkeletalMeshComponent=@n7, BoneName=upperarm_r, BaseName=clavicle_r, InputRange=(X=25,Y=64), TwistAxis="0,0,1")
   ) -> (Output_Get:double)
   > Set ADS_UP_R_Pistol(ADS_UP_R_Pistol:double=0) -> (Output_Get:double)
   n8 = Get Skel Mesh()
   Set Hood_ADS_UP_L_Assault(
    execute=@n6.then_2
    Hood_ADS_UP_L_Assault:double=GetAlphaFromReferenceTransformDelta(SkeletalMeshComponent=@n8, BoneName=upperarm_l, BaseName=clavicle_l, InputRange=(X=10,Y=50), TwistAxis="0,0,1")
   ) -> (Output_Get:double)
   > Set Hood_ADS_UP_L_Pistol(Hood_ADS_UP_L_Pistol:double=0) -> (Output_Get:double)
   Set Hood_ADS_UP_R_Assault(
    execute=@n6.then_3
    Hood_ADS_UP_R_Assault:double=GetAlphaFromReferenceTransformDelta(SkeletalMeshComponent=@n8, BoneName=upperarm_r, BaseName=clavicle_r, InputRange=(X=25,Y=64), TwistAxis="0,0,1")
   ) -> (Output_Get:double)
   > Set Hood_ADS_UP_R_Pistol(Hood_ADS_UP_R_Pistol:double=0) -> (Output_Get:double)
   Branch(execute=@n5.Pistol)
   > Set Alt_weapon(Alt_weapon=true)
   > n9 = Sequence()
   n10 = Get Skel Mesh()
   > Set ADS_UP_L_Pistol(ADS_UP_L_Pistol=GetAlphaFromReferenceTransformDelta(SkeletalMeshComponent=@n10, BoneName=upperarm_l, BaseName=clavicle_l, InputRange=(X=17,Y=75), TwistAxis="0,0,1"))
   > Set ADS_UP_L_Assault(ADS_UP_L_Assault=0)
   Set ADS_UP_R_Pistol(
    execute=@n9.then_1
    ADS_UP_R_Pistol=GetAlphaFromReferenceTransformDelta(SkeletalMeshComponent=@n10, BoneName=upperarm_r, BaseName=clavicle_r, InputRange=(X=45,Y=90), TwistAxis="0,0,1")
   )
   > Set ADS_UP_R_Assault(ADS_UP_R_Assault=0)
   n11 = Get Skel Mesh()
   Set Hood_ADS_UP_L_Pistol(
    execute=@n9.then_2
    Hood_ADS_UP_L_Pistol=GetAlphaFromReferenceTransformDelta(SkeletalMeshComponent=@n11, BoneName=upperarm_l, BaseName=clavicle_l, InputRange=(X=40,Y=83), TwistAxis="0,0,1")
   )
   > Set Hood_ADS_UP_L_Assault(Hood_ADS_UP_L_Assault=0)
   Set Hood_ADS_UP_R_Pistol(
    execute=@n9.then_3
    Hood_ADS_UP_R_Pistol=GetAlphaFromReferenceTransformDelta(SkeletalMeshComponent=@n11, BoneName=upperarm_r, BaseName=clavicle_r, InputRange=(X=50,Y=95), TwistAxis="0,0,1")
   )
   > Set Hood_ADS_UP_R_Assault(Hood_ADS_UP_R_Assault=0)
   n12 = Branch(execute=@n2.then_2, Condition=OR Boolean(A=Get bIsSkydiveFloating(), B=Get bIsSkydiveDiveMode()))
   > Set isSkyDiveNew(isSkyDiveNew:bool=true) -> (Output_Get:bool)
   Set isSkyDiveNew(execute=@n12.else, isSkyDiveNew=false)
   n13 = Gravity Override State Function DEPRECATED(
    execute=@n2.then_0
    CharacterPartAnimInstance=Self-Reference()
    GameTimeInSeconds=GetGameTimeInSeconds()
    GravityOverrideParams=Get Gravity Override Params()
   )
   > Set CapeGravityOverride(CapeGravityOverride:Vector=@n13) -> (Output_Get:Vector)
   Event BlueprintInitializeAnimation()
   n14 = TryGetPawnOwner()
   > Is Valid(InputObject=@n14)
   > n15 = Cast To FortPlayerPawnAthena(Object=@n14)
   > Set player_pawn(player_pawn:Fort Player Pawn Athena Object Reference=@n15) -> (Output_Get:Fort Player Pawn Athena Object Reference)
   > Set NoSkyDive(NoSkyDive=1)
   Set ADS_UP_L_Assault(ADS_UP_L_Assault=0)
   > Set ADS_UP_L_Pistol(ADS_UP_L_Pistol=0)
   TryGetPawnOwner()
   Sequence!(execute) -> (then_0, then_1)
   Set The_SkydiveFloat(The_SkydiveFloat=1)
   Set ADS_UP_R_Assault(ADS_UP_R_Assault=0)
   Set ADS_UP_R_Pistol(ADS_UP_R_Pistol=0)
   Set ADS_UP_R_Pistol(ADS_UP_R_Pistol=0)
   Set ADS_UP_L_Assault(ADS_UP_L_Assault=0)
   Set ADS_UP_R_Assault(ADS_UP_R_Assault=0)
  ]}
]
variables [
 ADS_UP_L_Assault {type=number}
 ADS_UP_L_Pistol {type=number}
 ADS_UP_R_Assault {type=number}
 ADS_UP_R_Pistol {type=number}
 Alt_weapon {type=boolean}
 Cape Trail Fake Velocity Mid Left B {type=Vector}
 CapeGravityOverride {type=Vector}
 CapeTrailFakeVelocityLowLeft_b {type=Vector}
 CapeTrailFakeVelocityLowRight_b {type=Vector}
 CapeTrailFakeVelocityMidLeft_b {type=Vector}
 CapeTrailFakeVelocityMidRight_b {type=Vector}
 CapeTrailRelaxSpeedScale {type=number, default=1.25}
 crouch alpha {type=number}
 crouch alpha inv {type=number}
 Gravity Override Params {type=GravityOverrideParamsStruct}
 Hood_ADS_UP_L_Assault {type=number}
 Hood_ADS_UP_L_Pistol {type=number}
 Hood_ADS_UP_R_Assault {type=number}
 Hood_ADS_UP_R_Pistol {type=number}
 isSkyDiveNew {type=boolean}
 IsSkydiving {type=number}
 isWearingBackbling {type=boolean}
 NoSkyDive {type=number}
 player_pawn {type=FortPlayerPawnAthena}
 Skel Mesh {type=SkeletalMeshComponent}
 SkydiveFloat {type=number}
 Target {type=CustomCharacterPartAnimInstance_Rebirth_Soldier_M, default=null}
 The_SkydiveFloat {type=number}
]
